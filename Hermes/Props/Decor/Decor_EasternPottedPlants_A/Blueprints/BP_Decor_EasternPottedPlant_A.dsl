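; Blueprint for an eastern-style potted plant decor prop, built with socket fillers.
; Its construction script runs the parent's, then reconstructs all socket fillers on the actor.
; It also acts as an exposed asset provider, adding version 1 and allowing assets referenced by the EasternPlant A, B and C variables.
; BeginPlay and Tick hold no active logic.
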(event EventBeginPlay
  (|CallParentFunction) ; This node is disabled and will not be called.\nDrag off pins to build functionality.
)

(event EventTick
  (params (DeltaSeconds :type Float%20%28single-precision%29)))

(fn BuildExposedAssetProviderVersion(Implementation)
  (params (Version :type Fort%20Exposed%20Asset%20Provider%20Version%20Structure%20%28by%20ref%29))
  (|CallParentFunction Version)
  (ExposedAssetProvider|AddVersion Version 1)
  (return))

(fn CollectAssetReferencestoAllow(Implementation)
  (params (Collector :type Fort%20Exposed%20Asset%20Reference%20Collector%20Structure%20%28by%20ref%29))
  (|CallParentFunction Collector)
  (ExposedAssetProvider|CollectObjectPropertyAssetReferences Collector 0 (Variables|Default|GetEasternPlantA))
  (ExposedAssetProvider|CollectObjectPropertyAssetReferences Collector 0 (Variables|Default|GetEasternPlantB))
  (ExposedAssetProvider|CollectObjectPropertyAssetReferences Collector 0 (Variables|Default|GetEasternPlantC))
  (return))

(fn ConstructionScript
  (|CallParentFunction)
  (Class|SocketFillerFunctionLibrary|ReconstructAllSocketFillers self))
